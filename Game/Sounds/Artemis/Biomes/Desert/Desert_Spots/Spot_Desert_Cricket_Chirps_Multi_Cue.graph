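# Sound cue for desert biome ambient one-shot cricket chirps, spatialized with an insect attenuation asset and played quietly in the outdoor ambient one-shot class.
# A quality-level root sends its first two slots to an equal-weight random pick of eight chirp waves and its last two to a smaller pick of four of those waves.
# Each random pick passes through its own modulator at default settings for slight pitch and volume variation.
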SoundCue /Game/Sounds/Artemis/Biomes/Desert/Desert_Spots/Spot_Desert_Cricket_Chirps_Multi_Cue
root=QualityLevel_0
nodes [
 {id=Modulator_12, children [Random_1]}
 {id=Modulator_2, children [Random_0]}
 {id=QualityLevel_0, children [Modulator_2,Modulator_2,Modulator_12,Modulator_12]}
 {
  id=Random_0
  children [WavePlayer_0,WavePlayer_1,WavePlayer_2,WavePlayer_3,WavePlayer_4,WavePlayer_5,WavePlayer_6,WavePlayer_7]
  properties {weights [1,1,1,1,1,1,1,1]}
 }
 {id=Random_1, children [WavePlayer_8, WavePlayer_9, WavePlayer_10, WavePlayer_11], properties {weights [1,1,1,1]}}
 {id=WavePlayer_0, properties {soundWaveAssetPtr=&/Spot_Desert_Cricket_Chirp_Multi_08}}
 {id=WavePlayer_1, properties {soundWaveAssetPtr=&/Spot_Desert_Cricket_Chirp_Multi_01}}
 {id=WavePlayer_10, properties {soundWaveAssetPtr=&/Spot_Desert_Cricket_Chirp_Multi_02}}
 {id=WavePlayer_11, properties {soundWaveAssetPtr=&/Spot_Desert_Cricket_Chirp_Multi_04}}
 {id=WavePlayer_2, properties {soundWaveAssetPtr=&Spot_Desert_Cricket_Chirp_Multi_02}}
 {id=WavePlayer_3, properties {soundWaveAssetPtr=&/Spot_Desert_Cricket_Chirp_Multi_03}}
 {id=WavePlayer_4, properties {soundWaveAssetPtr=&Spot_Desert_Cricket_Chirp_Multi_04}}
 {id=WavePlayer_5, properties {soundWaveAssetPtr=&/Spot_Desert_Cricket_Chirp_Multi_05}}
 {id=WavePlayer_6, properties {soundWaveAssetPtr=&/Spot_Desert_Cricket_Chirp_Multi_06}}
 {id=WavePlayer_7, properties {soundWaveAssetPtr=&/Spot_Desert_Cricket_Chirp_Multi_07}}
 {id=WavePlayer_8, properties {soundWaveAssetPtr=&Spot_Desert_Cricket_Chirp_Multi_01}}
 {id=WavePlayer_9, properties {soundWaveAssetPtr=&Spot_Desert_Cricket_Chirp_Multi_03}}
]
properties {
 attenuationSettings=/Game/Sounds/Attenuation/Fort_TileAmbient_OneShot_Insect_By_Att
 bEnableBusSends=false
 maxDistance=6000
 soundClassObject=&Fort_Ambient_OneShots_Outdoors
 volumeMultiplier=0.15
 preEffectBusSends [{}]
}
