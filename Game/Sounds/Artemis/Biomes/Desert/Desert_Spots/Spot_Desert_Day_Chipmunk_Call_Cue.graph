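# Sound cue for a desert daytime chipmunk call, an outdoor ambient one-shot using bird attenuation.
# The first two quality levels randomly play either one of six calls or that layered with a briefly delayed, higher-pitched second call.
# The other two quality levels play one of four calls with a wider, lower pitch range.
# Every branch randomizes pitch and volume, and output also feeds exterior ambient bus and reverb sends.
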SoundCue /Game/Sounds/Artemis/Biomes/Desert/Desert_Spots/Spot_Desert_Day_Chipmunk_Call_Cue
root=QualityLevel_0
nodes [
 {id=Delay_0, children [Modulator_0], properties {delayMax=0.2, delayMin=0.1}}
 {id=Mixer_0, children [Modulator_1,Delay_0], properties {inputVolume [1,1]}}
 {id=Modulator_0, children [Random_2], properties {pitchMax=1.2, pitchMin=1, volumeMax=1.1, volumeMin=0.9}}
 {id=Modulator_1, children [Random_0], properties {pitchMax=1.1, pitchMin=0.9, volumeMax=1.1, volumeMin=0.9}}
 {id=Modulator_2, children [Random_1], properties {pitchMax=1.1, pitchMin=0.5, volumeMax=1.1, volumeMin=0.9}}
 {id=QualityLevel_0, children [Random_3,Random_3,Modulator_2,Modulator_2]}
 {id=Random_0, children [WavePlayer_0,WavePlayer_1,WavePlayer_2,WavePlayer_3,WavePlayer_4,WavePlayer_5], properties {weights [1,1,1,1,1,1]}}
 {id=Random_1, children [WavePlayer_6,WavePlayer_7,WavePlayer_8,WavePlayer_9], properties {weights [1,1,1,1]}}
 {
  id=Random_2
  children [WavePlayer_10, WavePlayer_11, WavePlayer_12, WavePlayer_13, WavePlayer_14, WavePlayer_15]
  properties {weights [1,1,1,1,1,1]}
 }
 {id=Random_3, children [Modulator_1,Mixer_0], properties {weights [1,1]}}
 {id=WavePlayer_0, properties {soundWaveAssetPtr=&/Spot_Desert_Day_Chipmunk_Call_01}}
 {id=WavePlayer_1, properties {soundWaveAssetPtr=&/Spot_Desert_Day_Chipmunk_Call_02}}
 {id=WavePlayer_10, properties {soundWaveAssetPtr=&Spot_Desert_Day_Chipmunk_Call_01}}
 {id=WavePlayer_11, properties {soundWaveAssetPtr=&Spot_Desert_Day_Chipmunk_Call_02}}
 {id=WavePlayer_12, properties {soundWaveAssetPtr=&/Spot_Desert_Day_Chipmunk_Call_03}}
 {id=WavePlayer_13, properties {soundWaveAssetPtr=&/Spot_Desert_Day_Chipmunk_Call_04}}
 {id=WavePlayer_14, properties {soundWaveAssetPtr=&/Spot_Desert_Day_Chipmunk_Call_05}}
 {id=WavePlayer_15, properties {soundWaveAssetPtr=&/Spot_Desert_Day_Chipmunk_Call_06}}
 {id=WavePlayer_2, properties {soundWaveAssetPtr=&Spot_Desert_Day_Chipmunk_Call_03}}
 {id=WavePlayer_3, properties {soundWaveAssetPtr=&Spot_Desert_Day_Chipmunk_Call_04}}
 {id=WavePlayer_4, properties {soundWaveAssetPtr=&Spot_Desert_Day_Chipmunk_Call_05}}
 {id=WavePlayer_5, properties {soundWaveAssetPtr=&Spot_Desert_Day_Chipmunk_Call_06}}
 {id=WavePlayer_6, properties {soundWaveAssetPtr=&Spot_Desert_Day_Chipmunk_Call_01}}
 {id=WavePlayer_7, properties {soundWaveAssetPtr=&Spot_Desert_Day_Chipmunk_Call_03}}
 {id=WavePlayer_8, properties {soundWaveAssetPtr=&Spot_Desert_Day_Chipmunk_Call_02}}
 {id=WavePlayer_9, properties {soundWaveAssetPtr=&Spot_Desert_Day_Chipmunk_Call_04}}
]
properties {
 attenuationSettings=&Fort_TileAmbient_OneShot_Bird_Att
 maxDistance=8005
 soundClassObject=&Fort_Ambient_OneShots_Outdoors
 volumeMultiplier=0.501
 busSends [{
   audioBus=null
   maxSendDistance=6000
   minSendDistance=3000
   minSendLevel=0.6
   sendLevel=1
   soundSourceBus=&AmbientAudio_Exterior_SourceBus
  }]
 soundSubmixSends [{minSendLevel=0.6, soundSubmix=&Fortnite_Reverb_Default}]
}
