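# Sound Cue for desert biome grasshopper chirps, a quiet outdoor ambient one-shot spot sound using insect attenuation.
# A Quality Level root sends its first two levels and its last two levels to separate branches.
# Each branch runs a Modulator for pitch and volume variation over a weighted Random pick of five chirps, with the third at half weight.
# The branches reference same-named waves at different asset paths.
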SoundCue /Game/Sounds/Artemis/Biomes/Desert/Desert_Spots/Spot_Desert_Grasshopper_Chirps_Cue
root=QualityLevel_0
nodes [
 {id=Modulator_12, children [Random_3]}
 {id=Modulator_2, children [Random_2]}
 {id=QualityLevel_0, children [Modulator_2,Modulator_2,Modulator_12,Modulator_12]}
 {id=Random_2, children [WavePlayer_12, WavePlayer_13, WavePlayer_14, WavePlayer_15, WavePlayer_16], properties {weights [1,1,0.5,1,1]}}
 {id=Random_3, children [WavePlayer_17, WavePlayer_18, WavePlayer_19, WavePlayer_20, WavePlayer_21], properties {weights [1,1,0.5,1,1]}}
 {id=WavePlayer_12, properties {soundWaveAssetPtr=&/Spot_Desert_Grasshopper_01}}
 {id=WavePlayer_13, properties {soundWaveAssetPtr=&/Spot_Desert_Grasshopper_02}}
 {id=WavePlayer_14, properties {soundWaveAssetPtr=&/Spot_Desert_Grasshopper_03}}
 {id=WavePlayer_15, properties {soundWaveAssetPtr=&/Spot_Desert_Grasshopper_04}}
 {id=WavePlayer_16, properties {soundWaveAssetPtr=&/Spot_Desert_Grasshopper_05}}
 {id=WavePlayer_17, properties {soundWaveAssetPtr=&Spot_Desert_Grasshopper_01}}
 {id=WavePlayer_18, properties {soundWaveAssetPtr=&Spot_Desert_Grasshopper_02}}
 {id=WavePlayer_19, properties {soundWaveAssetPtr=&Spot_Desert_Grasshopper_03}}
 {id=WavePlayer_20, properties {soundWaveAssetPtr=&Spot_Desert_Grasshopper_04}}
 {id=WavePlayer_21, properties {soundWaveAssetPtr=&Spot_Desert_Grasshopper_05}}
]
properties {
 attenuationSettings=/Game/Sounds/Attenuation/Fort_TileAmbient_OneShot_Insect_By_Att
 bEnableBusSends=false
 maxDistance=6000
 soundClassObject=&Fort_Ambient_OneShots_Outdoors
 volumeMultiplier=0.125
 preEffectBusSends [{}]
}
