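# Sound cue for spot one-shot hummingbird calls in the desert biome, played as outdoor ambience with bird one-shot attenuation.
# A quality-level node at the root sends the first two levels to a random pick of seven call waves and the last two to a pick of three.
# Each random pool is equally weighted and passes through its own modulator.
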SoundCue /Game/Sounds/Artemis/Biomes/Desert/Desert_Spots/Spot_Desert_HummingBird_Call_Cue
root=QualityLevel_0
nodes [
 {id=Modulator_3, children [Random_4]}
 {id=Modulator_4, children [Random_5]}
 {id=QualityLevel_0, children [Modulator_3,Modulator_3,Modulator_4,Modulator_4]}
 {
  id=Random_4
  children [WavePlayer_10, WavePlayer_15, WavePlayer_16, WavePlayer_18, WavePlayer_19, WavePlayer_20, WavePlayer_21]
  properties {weights [1,1,1,1,1,1,1]}
 }
 {id=Random_5, children [WavePlayer_22, WavePlayer_23, WavePlayer_24], properties {weights [1,1,1]}}
 {id=WavePlayer_10, properties {soundWaveAssetPtr=&/Spot_Desert_Hummingbird_Call_01}}
 {id=WavePlayer_15, properties {soundWaveAssetPtr=&/Spot_Desert_Hummingbird_Call_02}}
 {id=WavePlayer_16, properties {soundWaveAssetPtr=&/Spot_Desert_Hummingbird_Call_03}}
 {id=WavePlayer_18, properties {soundWaveAssetPtr=&/Spot_Desert_Hummingbird_Call_05}}
 {id=WavePlayer_19, properties {soundWaveAssetPtr=&/Spot_Desert_Hummingbird_Call_06}}
 {id=WavePlayer_20, properties {soundWaveAssetPtr=&/Spot_Desert_Hummingbird_Call_07}}
 {id=WavePlayer_21, properties {soundWaveAssetPtr=&/Spot_Desert_Hummingbird_Call_08}}
 {id=WavePlayer_22, properties {soundWaveAssetPtr=&Spot_Desert_Hummingbird_Call_01}}
 {id=WavePlayer_23, properties {soundWaveAssetPtr=&Spot_Desert_Hummingbird_Call_02}}
 {id=WavePlayer_24, properties {soundWaveAssetPtr=&Spot_Desert_Hummingbird_Call_03}}
]
properties {
 attenuationSettings=&Fort_TileAmbient_OneShot_Bird_Att
 bEnableBusSends=false
 maxDistance=8005
 soundClassObject=&Fort_Ambient_OneShots_Outdoors
 volumeMultiplier=0.501
 preEffectBusSends [{}]
}
