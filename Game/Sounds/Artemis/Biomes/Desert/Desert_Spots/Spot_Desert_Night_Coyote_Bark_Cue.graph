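# Sound cue for a distant coyote bark, a one-shot outdoor ambient spot sound for the desert biome at night.
# A quality-level switch feeds the first two levels from five random, equally weighted bark waves and the last two from a three-wave pool.
# Each pool passes through a modulator for slight pitch and volume variation.
# Output uses distant-creature attenuation and a distance-scaled send to the exterior ambient source bus.
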SoundCue /Game/Sounds/Artemis/Biomes/Desert/Desert_Spots/Spot_Desert_Night_Coyote_Bark_Cue
root=QualityLevel_0
nodes [
 {id=Modulator_1, children [Random_0], properties {pitchMax=1.1, volumeMax=1.1, volumeMin=0.9}}
 {id=Modulator_2, children [Random_2], properties {pitchMax=1.1, volumeMax=1.1, volumeMin=0.9}}
 {id=QualityLevel_0, children [Modulator_1,Modulator_1,Modulator_2,Modulator_2]}
 {id=Random_0, children [WavePlayer_0,WavePlayer_1,WavePlayer_2,WavePlayer_3,WavePlayer_4], properties {weights [1,1,1,1,1]}}
 {id=Random_2, children [WavePlayer_7,WavePlayer_6,WavePlayer_5], properties {weights [1,1,1]}}
 {id=WavePlayer_0, properties {soundWaveAssetPtr=&/Spot_Desert_Night_Coyote_Bark_5}}
 {id=WavePlayer_1, properties {soundWaveAssetPtr=&/Spot_Desert_Night_Coyote_Bark_01}}
 {id=WavePlayer_2, properties {soundWaveAssetPtr=&/Spot_Desert_Night_Coyote_Bark_02}}
 {id=WavePlayer_3, properties {soundWaveAssetPtr=&/Spot_Desert_Night_Coyote_Bark_03}}
 {id=WavePlayer_4, properties {soundWaveAssetPtr=&/Spot_Desert_Night_Coyote_Bark_04}}
 {id=WavePlayer_5, properties {soundWaveAssetPtr=&Spot_Desert_Night_Coyote_Bark_04}}
 {id=WavePlayer_6, properties {soundWaveAssetPtr=&Spot_Desert_Night_Coyote_Bark_03}}
 {id=WavePlayer_7, properties {soundWaveAssetPtr=&Spot_Desert_Night_Coyote_Bark_02}}
]
properties {
 attenuationSettings=/Game/Sounds/Attenuation/Fort_TileAmbient_OneShot_Distant_Creature_Att
 maxDistance=8500
 soundClassObject=&Fort_Ambient_OneShots_Outdoors
 volumeMultiplier=0.331
 busSends [{
   audioBus=null
   maxSendDistance=6000
   minSendDistance=3000
   minSendLevel=0.6
   sendLevel=1
   soundSourceBus=&AmbientAudio_Exterior_SourceBus
  }]
}
