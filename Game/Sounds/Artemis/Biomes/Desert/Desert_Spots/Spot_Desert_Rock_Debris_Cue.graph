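# Desert biome ambient spot Sound Cue that plays positional one-shot rock debris sounds outdoors.
# A quality level root routes its first two and last two levels to separate branches that use the same four debris waves.
# Each branch picks one wave at random with equal weight, then a Modulator varies pitch and volume.
# A positional tile ambient attenuation asset and the outdoor ambient one-shot sound class govern playback.
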SoundCue /Game/Sounds/Artemis/Biomes/Desert/Desert_Spots/Spot_Desert_Rock_Debris_Cue
root=QualityLevel_0
nodes [
 {id=Modulator_0, children [Random_3], properties {pitchMax=1.1, pitchMin=0.7, volumeMax=1.1, volumeMin=0.9}}
 {id=Modulator_1, children [Random_1], properties {pitchMax=1.1, pitchMin=0.7, volumeMax=1.1, volumeMin=0.9}}
 {id=QualityLevel_0, children [Modulator_1,Modulator_1,Modulator_0,Modulator_0]}
 {id=Random_1, children [WavePlayer_0, WavePlayer_5, WavePlayer_6, WavePlayer_17], properties {weights [1,1,1,1]}}
 {id=Random_3, children [WavePlayer_20, WavePlayer_21, WavePlayer_22, WavePlayer_27], properties {weights [1,1,1,1]}}
 {id=WavePlayer_0, properties {soundWaveAssetPtr=/Game/Sounds/Apollo/Biomes/Desert/AMB_SS_S15_Desert_RockDebris_04}}
 {id=WavePlayer_17, properties {soundWaveAssetPtr=/Game/Sounds/Apollo/Biomes/Desert/AMB_SS_S15_Desert_RockDebris_03}}
 {id=WavePlayer_20, properties {soundWaveAssetPtr=&AMB_SS_S15_Desert_RockDebris_04}}
 {id=WavePlayer_21, properties {soundWaveAssetPtr=/Game/Sounds/Apollo/Biomes/Desert/AMB_SS_S15_Desert_RockDebris_01}}
 {id=WavePlayer_22, properties {soundWaveAssetPtr=/Game/Sounds/Apollo/Biomes/Desert/AMB_SS_S15_Desert_RockDebris_02}}
 {id=WavePlayer_27, properties {soundWaveAssetPtr=&AMB_SS_S15_Desert_RockDebris_03}}
 {id=WavePlayer_5, properties {soundWaveAssetPtr=&AMB_SS_S15_Desert_RockDebris_01}}
 {id=WavePlayer_6, properties {soundWaveAssetPtr=&AMB_SS_S15_Desert_RockDebris_02}}
]
properties {
 attenuationSettings=/Game/Sounds/Attenuation/Fort_TileAmbient_OneShot_MorePositional_Att
 bEnableBusSends=false
 maxDistance=8000
 soundClassObject=&Fort_Ambient_OneShots_Outdoors
 volumeMultiplier=1.504
 preEffectBusSends [{}]
}
