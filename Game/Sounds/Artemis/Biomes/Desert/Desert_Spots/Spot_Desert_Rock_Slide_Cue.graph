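# Sound cue for a rock slide spot ambient in the desert biome, reusing the mountain biome's rock slide waves as a distant outdoor one-shot.
# A quality-level root routes the first two levels to a random pick of seven waves and the last two to a reduced pick of three.
# Each branch passes through a modulator that varies pitch and volume.
# The cue uses a shared distant-creak attenuation and the outdoor ambient one-shot sound class.
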SoundCue /Game/Sounds/Artemis/Biomes/Desert/Desert_Spots/Spot_Desert_Rock_Slide_Cue
root=QualityLevel_0
nodes [
 {id=Modulator_1, children [Random_0], properties {pitchMax=1.1, volumeMax=1.1, volumeMin=0.9}}
 {id=Modulator_2, children [Random_1], properties {pitchMax=1.1, volumeMax=1.1, volumeMin=0.9}}
 {id=QualityLevel_0, children [Modulator_1,Modulator_1,Modulator_2,Modulator_2]}
 {
  id=Random_0
  children [WavePlayer_0,WavePlayer_1,WavePlayer_2,WavePlayer_3,WavePlayer_4,WavePlayer_5,WavePlayer_6]
  properties {weights [1,1,1,1,1,1,1]}
 }
 {id=Random_1, children [WavePlayer_7,WavePlayer_8,WavePlayer_9], properties {weights [1,1,1]}}
 {id=WavePlayer_0, properties {soundWaveAssetPtr=/Game/Sounds/Artemis/Biomes/Mountains/Mountain_Spots/Spot_Mountain_Rock_Slide_01}}
 {id=WavePlayer_1, properties {soundWaveAssetPtr=/Game/Sounds/Artemis/Biomes/Mountains/Mountain_Spots/Spot_Mountain_Rock_Slide_02}}
 {id=WavePlayer_2, properties {soundWaveAssetPtr=/Game/Sounds/Artemis/Biomes/Mountains/Mountain_Spots/Spot_Mountain_Rock_Slide_03}}
 {id=WavePlayer_3, properties {soundWaveAssetPtr=/Game/Sounds/Artemis/Biomes/Mountains/Mountain_Spots/Spot_Mountain_Rock_Slide_04}}
 {id=WavePlayer_4, properties {soundWaveAssetPtr=/Game/Sounds/Artemis/Biomes/Mountains/Mountain_Spots/Spot_Mountain_Rock_Slide_05}}
 {id=WavePlayer_5, properties {soundWaveAssetPtr=/Game/Sounds/Artemis/Biomes/Mountains/Mountain_Spots/Spot_Mountain_Rock_Slide_06}}
 {id=WavePlayer_6, properties {soundWaveAssetPtr=/Game/Sounds/Artemis/Biomes/Mountains/Mountain_Spots/Spot_Mountain_Rock_Slide_07}}
 {id=WavePlayer_7, properties {soundWaveAssetPtr=&Spot_Mountain_Rock_Slide_01}}
 {id=WavePlayer_8, properties {soundWaveAssetPtr=&Spot_Mountain_Rock_Slide_03}}
 {id=WavePlayer_9, properties {soundWaveAssetPtr=&Spot_Mountain_Rock_Slide_06}}
]
properties {
 attenuationSettings=/Game/Sounds/Attenuation/Fort_TileAmbient_OneShot_Distant_Creak_Att
 bEnableBusSends=false
 maxDistance=8200
 soundClassObject=&Fort_Ambient_OneShots_Outdoors
 volumeMultiplier=0.501
 preEffectBusSends [{}]
}
