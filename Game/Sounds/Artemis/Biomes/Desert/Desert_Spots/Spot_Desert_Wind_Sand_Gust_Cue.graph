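# Sound cue for one-shot desert wind and sand gust ambience, using outdoor ambient one-shot sound class and tile wind attenuation.
# A quality switch sends its first two levels to a weighted random pick favoring ten slow sand gusts over five wind gusts.
# Its last two levels use a reduced pool of three wind gusts.
# Modulators on each pool boost volume and randomize pitch to vary repeats.
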SoundCue /Game/Sounds/Artemis/Biomes/Desert/Desert_Spots/Spot_Desert_Wind_Sand_Gust_Cue
root=QualityLevel_0
nodes [
 {id=Modulator_0, children [Random_2], properties {volumeMax=2.1, volumeMin=2}}
 {id=Modulator_1, children [Random_0], properties {pitchMax=1.1, volumeMax=2.2, volumeMin=2}}
 {id=Modulator_2, children [Random_1], properties {pitchMax=1.1, volumeMax=2.2, volumeMin=2}}
 {id=QualityLevel_0, children [Random_3,Random_3,Modulator_2,Modulator_2]}
 {id=Random_0, children [WavePlayer_0,WavePlayer_1,WavePlayer_2,WavePlayer_3,WavePlayer_4], properties {weights [1,1,1,1,1]}}
 {id=Random_1, children [WavePlayer_5,WavePlayer_6,WavePlayer_7], properties {weights [1,1,1]}}
 {
  id=Random_2
  children [
   WavePlayer_8
   WavePlayer_9
   WavePlayer_10
   WavePlayer_11
   WavePlayer_12
   WavePlayer_13
   WavePlayer_14
   WavePlayer_15
   WavePlayer_16
   WavePlayer_17
  ]
  properties {weights [1,1,1,1,1,1,1,1,1,1]}
 }
 {id=Random_3, children [Modulator_0,Modulator_1], properties {weights [1,0.5]}}
 {id=WavePlayer_0, properties {soundWaveAssetPtr=&/Spot_Desert_Wind_Sand_Gust_01}}
 {id=WavePlayer_1, properties {soundWaveAssetPtr=&/Spot_Desert_Wind_Sand_Gust_02}}
 {id=WavePlayer_10, properties {soundWaveAssetPtr=&/Spot_Desert_Slow_Sand_Gust_02}}
 {id=WavePlayer_11, properties {soundWaveAssetPtr=&/Spot_Desert_Slow_Sand_Gust_03}}
 {id=WavePlayer_12, properties {soundWaveAssetPtr=&/Spot_Desert_Slow_Sand_Gust_04}}
 {id=WavePlayer_13, properties {soundWaveAssetPtr=&/Spot_Desert_Slow_Sand_Gust_05}}
 {id=WavePlayer_14, properties {soundWaveAssetPtr=&/Spot_Desert_Slow_Sand_Gust_06}}
 {id=WavePlayer_15, properties {soundWaveAssetPtr=&/Spot_Desert_Slow_Sand_Gust_07}}
 {id=WavePlayer_16, properties {soundWaveAssetPtr=&/Spot_Desert_Slow_Sand_Gust_08}}
 {id=WavePlayer_17, properties {soundWaveAssetPtr=&/Spot_Desert_Slow_Sand_Gust_09}}
 {id=WavePlayer_2, properties {soundWaveAssetPtr=&/Spot_Desert_Wind_Sand_Gust_03}}
 {id=WavePlayer_3, properties {soundWaveAssetPtr=&/Spot_Desert_Wind_Sand_Gust_04}}
 {id=WavePlayer_4, properties {soundWaveAssetPtr=&/Spot_Desert_Wind_Sand_Gust_05}}
 {id=WavePlayer_5, properties {soundWaveAssetPtr=&Spot_Desert_Wind_Sand_Gust_02}}
 {id=WavePlayer_6, properties {soundWaveAssetPtr=&Spot_Desert_Wind_Sand_Gust_01}}
 {id=WavePlayer_7, properties {soundWaveAssetPtr=&Spot_Desert_Wind_Sand_Gust_05}}
 {id=WavePlayer_8, properties {soundWaveAssetPtr=&/Spot_Desert_Slow_Sand_Gust_10}}
 {id=WavePlayer_9, properties {soundWaveAssetPtr=&/Spot_Desert_Slow_Sand_Gust_01}}
]
properties {
 attenuationSettings=/Game/Sounds/Attenuation/Fort_TileAmbient_OneShot_Wind_Att
 bEnableBusSends=false
 maxDistance=8000
 soundClassObject=&Fort_Ambient_OneShots_Outdoors
 volumeMultiplier=0.5
 preEffectBusSends [{}]
}
